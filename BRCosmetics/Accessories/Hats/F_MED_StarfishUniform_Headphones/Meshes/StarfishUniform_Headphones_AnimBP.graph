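# Animation Blueprint for the StarfishUniform headphones hat cosmetic, adding secondary motion to its dynamic hair bones.
# Three Anim Dynamics chains (center, left, right dyn_hair bones) run in series, then convert to local space for the output pose.
# Each simulates in actor space with angular springs, high damping and limited swing, with the right chain's joint offset mirrored.
# The event graph holds only unconnected default nodes, so no variables or logic drive it.
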AnimBlueprint /BRCosmetics/Accessories/Hats/F_MED_StarfishUniform_Headphones/Meshes/StarfishUniform_Headphones_AnimBP
target_skeleton=&/StarfishUniform_Headphones_Skeleton
graphs [
 AnimGraph {nodes [
   n1 = "Anim Dynamics - Chain: dyn_hair_01 -> dyn_hair_02"(Alpha=1, ExternalForce="0,0,0") {
    angularDampingOverride=0.98
    angularSpringConstant=100
    bAngularSpring=true
    bChain=true
    bGravityOverrideInSimSpace=true
    bOverrideAngularDamping=true
    numSolverIterationsPostUpdate=4
    numSolverIterationsPreUpdate=16
    simulationSpace=Actor
    alphaBoolBlend {blendOption=Linear}
    boundBone {boneName=dyn_hair_01}
    chainEnd {boneName=dyn_hair_02}
    physicsBodyDefinitions [{
      sphereCollisionRadius=0
      boundBone {boneName=dyn_hair_01}
      boxExtents=<20 30 30>
      constraintSetup {angularTargetAxis=AxisZ, angularLimitsMax=<5 20 0>, angularLimitsMin=<-5 -20 -20>, angularTarget=<0 0 1>}
      localJointOffset=<2 0 0>
     }]
   }
   n2 = "Anim Dynamics - Chain: dyn_hair_01_l -> dyn_hair_02_l"(ComponentPose=@n1, Alpha=1, ExternalForce="0,0,0") {
    angularDampingOverride=0.98
    angularSpringConstant=100
    bAngularSpring=true
    bChain=true
    bGravityOverrideInSimSpace=true
    bOverrideAngularDamping=true
    numSolverIterationsPostUpdate=4
    numSolverIterationsPreUpdate=16
    simulationSpace=Actor
    alphaBoolBlend {blendOption=Linear}
    boundBone {boneName=dyn_hair_01_l}
    chainEnd {boneName=dyn_hair_02_l}
    physicsBodyDefinitions [{
      sphereCollisionRadius=0
      boundBone {boneName=dyn_hair_01_l}
      boxExtents=<20 30 30>
      constraintSetup {angularTargetAxis=AxisZ, angularLimitsMax=<5 20 0>, angularLimitsMin=<-5 -20 -20>, angularTarget=<0 0 1>}
      localJointOffset=<2 0 0>
     }]
   }
   n3 = "Anim Dynamics - Chain: dyn_hair_01_r -> dyn_hair_02_r"(ComponentPose=@n2, Alpha=1, ExternalForce="0,0,0") {
    angularDampingOverride=0.98
    angularSpringConstant=100
    bAngularSpring=true
    bChain=true
    bGravityOverrideInSimSpace=true
    bOverrideAngularDamping=true
    numSolverIterationsPostUpdate=4
    numSolverIterationsPreUpdate=16
    simulationSpace=Actor
    alphaBoolBlend {blendOption=Linear}
    boundBone {boneName=dyn_hair_01_r}
    chainEnd {boneName=dyn_hair_02_r}
    physicsBodyDefinitions [{
      sphereCollisionRadius=0
      boundBone {boneName=dyn_hair_01_r}
      boxExtents=<20 30 30>
      constraintSetup {angularTargetAxis=AxisZ, angularLimitsMax=<5 20 0>, angularLimitsMin=<-5 -20 -20>, angularTarget=<0 0 1>}
      localJointOffset=<-2 0 0>
     }]
   }
   Output Pose(Result=Component To Local(ComponentPose=@n3)) {layerGroup=None}
  ]}
 EventGraph {nodes [
   TryGetPawnOwner()
   Event BlueprintUpdateAnimation()
  ]}
]
variables []
